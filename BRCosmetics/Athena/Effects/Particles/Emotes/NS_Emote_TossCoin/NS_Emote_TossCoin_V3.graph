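# Niagara effect for the TossCoin emote, showing a holographic coin with projection rays, flares and sparks.
# User.CoinSide drives the burst counts of the front and back hologram mesh emitters, choosing the face shown.
# The coin follows bone hand_l with a curved rise-and-fall offset, plus animated scale, alpha, material parameters and a light.
# Mesh rays and a flare accompany it, then late flares and curl-noise sparks finish the effect.
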NiagaraSystem /BRCosmetics/Athena/Effects/Particles/Emotes/NS_Emote_TossCoin/NS_Emote_TossCoin_V3
dependencies {
 usedDataInterfaces [&NiagaraDataInterfaceSkeletalMesh, &NiagaraDataInterfaceVectorCurve, &NiagaraDataInterfaceCurve]
 usedDynamicInputs []
 usedModules [&SystemState]
}
emitters {
 CameraFacing_Coin_Hologram_BackSide {
  emitterUpdateScript [
   &EmitterState()
   &SpawnBurst_Instantaneous(
    Spawn Count=&MakeIntFromFloat(Float=&Lerp_Float(A=0, B=1, Alpha=$User.CoinSide))
    Spawn Time=0.0417
    Use Spawn Probability=false
   )
  ]
  particleSpawnScript [/Niagara/Modules/Spawn/Initialization/InitializeParticle(
    Lifetime=1.35
    Mesh Scale=&VectorFromFloat(Value=2.5)
    Module.Write Color=true
    Module.Write Lifetime=true
    Module.Write Mass=true
    Module.Write Material Random=true
    Module.Write Position=true
    Module.Write Scale=true
    Module.Write SpriteRotation=true
    Module.Write SpriteSize=true
    Position=&SimulationPosition()
    Sprite Rotation=&UniformRangedFloat(Minimum=0, Maximum=0)
    Sprite Size=<100 100>
   )]
  particleUpdateScript [
   &ParticleState()
   &SolveForcesAndVelocity(Clamp Velocity=false, Limit Acceleration=false)
   &SkeletalMeshLocation(
    Bone Sampling Mode=::NewEnumerator4
    Inherit Velocity=false
    MaskSpawnGroup=false
    Sampled Position Offset=&VectorFromCurve(VectorCurve=DataInterface() {propertyValues {
       XCurve {keys [<at -0.0014 0>]}
       YCurve {keys [<at -0.0011 0>]}
       ZCurve {keys [
         {arriveTangent=404.1048, interpMode=RCIM_Cubic, leaveTangent=404.1069, tangentMode=RCTM_User, time=0.0062, value=45.72}
         {arriveTangent=72.5222, interpMode=RCIM_Cubic, leaveTangent=72.5222, time=0.1288, value=70.6533}
         <at 0.4927 81>
         {arriveTangent=-174.8625, interpMode=RCIM_Cubic, leaveTangent=-174.8625, tangentMode=RCTM_Break, time=0.9124, value=77.3929}
         <at 0.9538 0.3587>
        ]}
      }})
    Skeletal Mesh=DataInterface() {propertyValues {
      DefaultMesh=null
      ExcludeBoneName=root
      OptionalComponentClass=null
      PreviewMesh=/Game/Characters/Player/Male/Male_Avg_Base/M_Med_Base_Skeleton
      SkinningMode=SkinOnTheFly
      SoftSourceActor=null
      SourceMode=Default
      UvSetIndex=0
      WholeMeshLOD=-1
      bExcludeBone=false
      bReadDeformedGeometry=true
      bRequireCurrentFrameData=true
      ComponentTags []
      FilteredBones [hand_l]
      FilteredSockets []
      MeshUserParameter {parameter {name=None, typeDefHandle {registeredTypeIndex=1}, varData []}}
      SamplingRegions []
     }}
   )
   &ScaleMeshSize(Scale Factor=&MakeVector(
     X=1
     Y=&FloatFromCurve(FloatCurve=DataInterface() {propertyValues {Curve {keys [
          <at 0 0.5>
          <at 0.0833 1.0079>
          {arriveTangent=-1.096, interpMode=RCIM_Cubic, leaveTangent=-7.2979, tangentMode=RCTM_Break, time=0.8822, value=0.9489}
          <at 0.9513 0.0237>
         ]}}})
     Z=&FloatFromCurve(FloatCurve=DataInterface() {propertyValues {Curve {keys [
          <at 0 0.4955>
          <at 0.0955 1>
          {arriveTangent=-1.1021, interpMode=RCIM_Cubic, leaveTangent=0.4162, tangentMode=RCTM_Break, time=0.8964, value=1}
          {arriveTangent=-6.3724, interpMode=RCIM_Cubic, leaveTangent=-6.3724, tangentMode=RCTM_User, time=1.0011, value=0.002}
         ]}}})
    ))
   &Color(Scale Alpha=&FloatFromCurve(FloatCurve=DataInterface() {propertyValues {Curve {keys [
         <at -0.0046 1.0113>
         {arriveTangent=-5.1777, interpMode=RCIM_Cubic, leaveTangent=-5.1777, tangentMode=RCTM_User, time=0.8233, value=0.921}
         <at 1 0>
        ]}}}))
   /SetVariables(Particles.LightRadius=&Lerp_Float(
     A=10
     B=100
     Alpha=&FloatFromCurve(FloatCurve=DataInterface() {propertyValues {Curve {keys [<at -0.0046 1.0113>,<at 0.931 0.9859>,<at 1.0011 0.002>]}}})
    )) {isSetParameters=true, module=SetVariables_#1}
   &DynamicMaterialParameters(
    Index 0 Param 1=&FloatFromCurve(FloatCurve=DataInterface() {propertyValues {Curve {keys [
         {arriveTangent=-0.9895, interpMode=RCIM_Cubic, leaveTangent=-4.4715, tangentMode=RCTM_User, time=0, value=1}
         {arriveTangent=-0.7274, interpMode=RCIM_Cubic, leaveTangent=-0.7274, time=0.305, value=0.0626}
         {arriveTangent=-0.146, interpMode=RCIM_Cubic, leaveTangent=-0.146, time=0.8712, value=0.3663}
         <at 1.3259 -0.0865>
        ]}}})
    Index 0 Param 2=&FloatFromCurve(FloatCurve=DataInterface() {propertyValues {Curve {keys [
         <at -0.0046 0 RCIM_Cubic>
         <at 0.311 0.1069>
         {arriveTangent=-0.0948, interpMode=RCIM_Cubic, leaveTangent=-0.0948, time=0.7649, value=0.0563}
         <at 0.994 0.0422>
        ]}}})
    Index 0 Param 3=&FloatFromCurve(FloatCurve=DataInterface() {propertyValues {Curve {keys [
         {arriveTangent=-2.2759, interpMode=RCIM_Cubic, leaveTangent=-2.2759, tangentMode=RCTM_User, time=-0.0046, value=0.4}
         <at 0.2273 0.0157>
         <at 0.5712 -0.0041>
        ]}}})
    Module.Param0WriteEnabled=true
    Module.Param1WriteEnabled=true
    Module.Param2WriteEnabled=true
    Module.Param3WriteEnabled=true
   )
  ]
  rendererClasses [&NiagaraMeshRendererProperties, &NiagaraLightRendererProperties]
  renderers [
   {rendererClass=&NiagaraMeshRendererProperties, rendererIndex=0}
   {rendererClass=&NiagaraLightRendererProperties, rendererIndex=1}
  ]
 }
 CameraFacing_Coin_Hologram_FrontSide {
  emitterUpdateScript [
   &EmitterState()
   &SpawnBurst_Instantaneous(
    Spawn Count=&MakeIntFromFloat(Float=&Lerp_Float(A=1, B=0, Alpha=$User.CoinSide))
    Spawn Time=0.0417
    Use Spawn Probability=false
   )
  ]
  particleSpawnScript [&InitializeParticle(
    Lifetime=1.35
    Mesh Scale=&VectorFromFloat(Value=2.5)
    Module.Write Color=true
    Module.Write Lifetime=true
    Module.Write Mass=true
    Module.Write Material Random=true
    Module.Write Position=true
    Module.Write Scale=true
    Module.Write SpriteRotation=true
    Module.Write SpriteSize=true
    Position=&SimulationPosition()
    Sprite Rotation=&UniformRangedFloat(Minimum=0, Maximum=0)
    Sprite Size=<100 100>
   )]
  particleUpdateScript [
   &ParticleState()
   &SolveForcesAndVelocity(Clamp Velocity=false, Limit Acceleration=false)
   &SkeletalMeshLocation(
    Bone Sampling Mode=::NewEnumerator4
    Inherit Velocity=false
    MaskSpawnGroup=false
    Sampled Position Offset=&VectorFromCurve(VectorCurve=DataInterface() {propertyValues {
       XCurve {keys [<at -0.0014 0>]}
       YCurve {keys [<at -0.0011 0>]}
       ZCurve {keys [
         {arriveTangent=404.1048, interpMode=RCIM_Cubic, leaveTangent=404.1069, tangentMode=RCTM_User, time=0.0062, value=45.72}
         {arriveTangent=72.5222, interpMode=RCIM_Cubic, leaveTangent=72.5222, time=0.1288, value=70.6533}
         <at 0.4927 81>
         {arriveTangent=-174.8625, interpMode=RCIM_Cubic, leaveTangent=-174.8625, tangentMode=RCTM_Break, time=0.9124, value=77.3929}
         <at 0.9538 0.3587>
        ]}
      }})
    Skeletal Mesh=DataInterface() {propertyValues {
      DefaultMesh=null
      ExcludeBoneName=root
      OptionalComponentClass=null
      PreviewMesh=&M_Med_Base_Skeleton
      SkinningMode=SkinOnTheFly
      SoftSourceActor=null
      SourceMode=Default
      UvSetIndex=0
      WholeMeshLOD=-1
      bExcludeBone=false
      bReadDeformedGeometry=true
      bRequireCurrentFrameData=true
      ComponentTags []
      FilteredBones [hand_l]
      FilteredSockets []
      MeshUserParameter {parameter {name=None, typeDefHandle {registeredTypeIndex=1}, varData []}}
      SamplingRegions []
     }}
   )
   &ScaleMeshSize(Scale Factor=&MakeVector(
     X=1
     Y=&FloatFromCurve(FloatCurve=DataInterface() {propertyValues {Curve {keys [
          <at 0 0.5>
          <at 0.0833 1.0079>
          {arriveTangent=-1.096, interpMode=RCIM_Cubic, leaveTangent=-7.2979, tangentMode=RCTM_Break, time=0.8822, value=0.9489}
          <at 0.9513 0.0237>
         ]}}})
     Z=&FloatFromCurve(FloatCurve=DataInterface() {propertyValues {Curve {keys [
          <at 0 0.4955>
          <at 0.0955 1>
          {arriveTangent=-1.1021, interpMode=RCIM_Cubic, leaveTangent=0.4162, tangentMode=RCTM_Break, time=0.8964, value=1}
          {arriveTangent=-6.3724, interpMode=RCIM_Cubic, leaveTangent=-6.3724, tangentMode=RCTM_User, time=1.0011, value=0.002}
         ]}}})
    ))
   &Color(Scale Alpha=&FloatFromCurve(FloatCurve=DataInterface() {propertyValues {Curve {keys [
         <at -0.0046 1.0113>
         {arriveTangent=-4.1555, interpMode=RCIM_Cubic, leaveTangent=-4.1555, tangentMode=RCTM_User, time=0.8133, value=0.9117}
         <at 1 0>
        ]}}}))
   /SetVariables(Particles.LightRadius=&Lerp_Float(
     A=10
     B=100
     Alpha=&FloatFromCurve(FloatCurve=DataInterface() {propertyValues {Curve {keys [<at -0.0046 1.0113>,<at 0.931 0.9859>,<at 1.0011 0.002>]}}})
    )) {isSetParameters=true, module=SetVariables_#1}
   &DynamicMaterialParameters(
    Index 0 Param 1=&FloatFromCurve(FloatCurve=DataInterface() {propertyValues {Curve {keys [
         {arriveTangent=-0.6795, interpMode=RCIM_Cubic, leaveTangent=-3.6997, tangentMode=RCTM_User, time=0, value=1}
         {arriveTangent=-0.412, interpMode=RCIM_Cubic, leaveTangent=-0.412, tangentMode=RCTM_User, time=1.0091, value=0}
        ]}}})
    Index 0 Param 2=&FloatFromCurve(FloatCurve=DataInterface() {propertyValues {Curve {keys [<at -0.0046 0 RCIM_Cubic>,<at 0.996 -0.0333>]}}})
    Index 0 Param 3=&FloatFromCurve(
     FloatCurve=DataInterface() {propertyValues {Curve {keys [
         {arriveTangent=-2.2759, interpMode=RCIM_Cubic, leaveTangent=-2.2759, tangentMode=RCTM_User, time=-0.0046, value=0.4}
         {arriveTangent=-3.5698, interpMode=RCIM_Cubic, leaveTangent=-3.5698, tangentMode=RCTM_User, time=0.5673, value=0}
        ]}}}
     Scale Curve=0
    )
    Module.Param0WriteEnabled=true
    Module.Param1WriteEnabled=true
    Module.Param2WriteEnabled=true
    Module.Param3WriteEnabled=true
   )
  ]
  rendererClasses [&NiagaraMeshRendererProperties, &NiagaraLightRendererProperties]
  renderers [
   {rendererClass=&NiagaraMeshRendererProperties, rendererIndex=0}
   {rendererClass=&NiagaraLightRendererProperties, rendererIndex=1}
  ]
 }
 Flare {
  emitterUpdateScript [
   &EmitterState()
   &SpawnBurst_Instantaneous(Spawn Count=1, Use Spawn Probability=false)
  ]
  particleSpawnScript [&InitializeParticle(
    Lifetime=1.2
    Mesh Scale=&UniformRangedVector(Minimum=<0.7 0.7 0.7>, Maximum=<0.9 0.7 0.7>)
    Module.Write Color=true
    Module.Write Lifetime=true
    Module.Write Mass=true
    Module.Write Material Random=true
    Module.Write Position=true
    Module.Write Scale=false
    Module.Write SpriteRotation=false
    Module.Write SpriteSize=true
    Position=&SimulationPosition()
    Sprite Rotation=&UniformRangedFloat(Minimum=0, Maximum=360)
    Sprite Size=<120 50>
   )]
  particleUpdateScript [
   &ParticleState()
   &SolveForcesAndVelocity(Clamp Velocity=false, Limit Acceleration=false)
   &ScaleSpriteSize(Scale Factor=&Vector2DFromFloat(Value=1.3))
   &CameraOffset(Camera Offset Amount=20)
   &Color(Scale Alpha=&UniformRangedFloat(Minimum=0.1, Maximum=1))
  ]
  rendererClasses [&NiagaraSpriteRendererProperties]
  renderers [{rendererClass=&NiagaraSpriteRendererProperties}]
 }
 Flare001 {
  emitterUpdateScript [
   &EmitterScalability(
    Module.bApplySpawnCountScale=false
    Module.bApplySpawnCountScaleByDistanceFraction=false
    Module.bUseMaxDistance=false
    Module.bUseMinDistance=false
   ) {enabled=false}
   &EmitterLifeCycle(NextLoopDuration=2) {enabled=false}
   &EmitterState()
   &SpawnBurst_Instantaneous(Spawn Count=1, Spawn Time=1.3125, Use Spawn Probability=false)
  ]
  particleSpawnScript [&InitializeParticle(
    Lifetime=0.15
    Mesh Scale=&UniformRangedVector(Minimum=<0.7 0.7 0.7>, Maximum=<0.9 0.7 0.7>)
    Module.Write Color=true
    Module.Write Lifetime=true
    Module.Write Mass=true
    Module.Write Material Random=true
    Module.Write Position=true
    Module.Write Scale=false
    Module.Write SpriteRotation=false
    Module.Write SpriteSize=true
    Position=&SimulationPosition()
    Sprite Rotation=&UniformRangedFloat(Minimum=0, Maximum=360)
    Sprite Size=<400 200>
   )]
  particleUpdateScript [
   &ParticleState()
   &SolveForcesAndVelocity(Clamp Velocity=false, Limit Acceleration=false)
   &ScaleSpriteSize(Scale Factor=&Vector2DFromFloat(Value=&FloatFromCurve(FloatCurve=DataInterface() {propertyValues {Curve {keys [
          <at -0.0046 1.0113 RCIM_Cubic>
          {arriveTangent=-0.0644, interpMode=RCIM_Cubic, leaveTangent=-0.0644, tangentMode=RCTM_User, time=1.0011, value=0.002}
         ]}}})))
   &CameraOffset(Camera Offset Amount=20)
   &Color(Color=<rgba 1 1 1 0.15>)
  ]
  rendererClasses [&NiagaraSpriteRendererProperties]
  renderers [{rendererClass=&NiagaraSpriteRendererProperties}]
 }
 Flare002 {
  emitterUpdateScript [
   &EmitterScalability(
    Module.bApplySpawnCountScale=false
    Module.bApplySpawnCountScaleByDistanceFraction=false
    Module.bUseMaxDistance=false
    Module.bUseMinDistance=false
   ) {enabled=false}
   &EmitterLifeCycle(NextLoopDuration=2) {enabled=false}
   &EmitterState()
   &SpawnBurst_Instantaneous(Spawn Count=1, Spawn Time=1.3833, Use Spawn Probability=false)
  ]
  particleSpawnScript [&InitializeParticle(
    Color=<rgba 1 1 1 0.2>
    Lifetime=0.12
    Mesh Scale=&UniformRangedVector(Minimum=<0.7 0.7 0.7>, Maximum=<0.9 0.7 0.7>)
    Module.Write Color=true
    Module.Write Lifetime=true
    Module.Write Mass=true
    Module.Write Material Random=true
    Module.Write Position=true
    Module.Write Scale=false
    Module.Write SpriteRotation=false
    Module.Write SpriteSize=true
    Position=&SimulationPosition()
    Sprite Rotation=&UniformRangedFloat(Minimum=0, Maximum=360)
    Sprite Size=<100 500>
   )]
  particleUpdateScript [
   &ParticleState()
   &SolveForcesAndVelocity(Clamp Velocity=false, Limit Acceleration=false)
   &ScaleSpriteSize(Scale Factor=&Vector2DFromFloat(Value=&FloatFromCurve(FloatCurve=DataInterface() {propertyValues {Curve {keys [
          {arriveTangent=-0.2021, interpMode=RCIM_Cubic, leaveTangent=-0.2021, tangentMode=RCTM_User, time=-0.0046, value=1.0113}
          {arriveTangent=-0.0644, interpMode=RCIM_Cubic, leaveTangent=-0.0644, tangentMode=RCTM_User, time=1.0011, value=0.002}
         ]}}})))
   &CameraOffset(Camera Offset Amount=20)
   &Color(Color=<rgba 1 1 1 0.15>)
  ]
  rendererClasses [&NiagaraSpriteRendererProperties]
  renderers [{rendererClass=&NiagaraSpriteRendererProperties}]
 }
 TinySparks {
  emitterUpdateScript [
   &EmitterScalability(
    Module.bApplySpawnCountScale=false
    Module.bApplySpawnCountScaleByDistanceFraction=false
    Module.bUseMaxDistance=false
    Module.bUseMinDistance=false
   ) {enabled=false}
   &EmitterLifeCycle(NextLoopDuration=2) {enabled=false}
   &EmitterState()
   &SpawnBurst_Instantaneous(Spawn Count=3, Spawn Time=1.3833, Use Spawn Probability=false)
  ]
  particleSpawnScript [
   &InitializeParticle(
    Color=<rgba 0 0.3136 1 1>
    Lifetime=0.5
    Mesh Scale=&UniformRangedVector(Minimum=<0.7 0.7 0.7>, Maximum=<0.9 0.7 0.7>)
    Module.Write Color=true
    Module.Write Lifetime=true
    Module.Write Mass=true
    Module.Write Material Random=true
    Module.Write Position=true
    Module.Write Scale=false
    Module.Write SpriteRotation=false
    Module.Write SpriteSize=true
    Position=&SimulationPosition()
    Sprite Rotation=&UniformRangedFloat(Minimum=0, Maximum=360)
    Sprite Size=<2 6>
   )
   &AddVelocity(Coordinate Space=::World, Velocity=<0 0 -500>)
   /Niagara/Modules/Spawn/Location/SphereLocation(MaskSpawnGroup=false, Module.Override Randomness=false, Sphere Radius=10, Surface Only=false)
  ]
  particleUpdateScript [
   &ParticleState()
   &Drag(Drag=10, Use Linear Drag=true, Use Rotational Drag=false)
   /Niagara/Modules/Update/Forces/CurlNoiseForce(Noise Frequency=0.1, Noise Strength=2500, Pan Noise=false)
   &SolveForcesAndVelocity(Clamp Velocity=false, Limit Acceleration=false)
   &ScaleSpriteSize(Scale Factor=&Vector2DFromFloat(Value=&FloatFromCurve(FloatCurve=DataInterface() {propertyValues {Curve {keys [
          {arriveTangent=-0.2021, interpMode=RCIM_Cubic, leaveTangent=-0.2021, tangentMode=RCTM_User, time=-0.0046, value=1.0113}
          {arriveTangent=-0.0644, interpMode=RCIM_Cubic, leaveTangent=-0.0644, tangentMode=RCTM_User, time=1.0011, value=0.002}
         ]}}})))
   &CameraOffset(Camera Offset Amount=20)
   &Color(Color=<rgba 1 1 1 0.45>)
  ]
  rendererClasses [&NiagaraSpriteRendererProperties]
  renderers [{rendererClass=&NiagaraSpriteRendererProperties}]
 }
 VelocityFacing_Projection_Ray {
  emitterUpdateScript [
   &EmitterState(Inactive Response=::NewEnumerator1, Life Cycle Mode=::NewEnumerator1, Loop Behavior=::NewEnumerator1, Loop Duration=2.1, UseLoopDelay=false)
   &SpawnRate(
    SpawnRate=&FloatFromCurve(
     FloatCurve=DataInterface() {propertyValues {Curve {keys [<at 0.496 1.0113>,<at 0.55 0>]}}}
     CurveIndex=$Emitter.NormalizedLoopAge
     Scale Curve=20
    )
    Use Spawn Probability=false
   )
  ]
  particleSpawnScript [
   &InitializeParticle(
    Color=<rgba 0.5 0.5 0.5 0.5>
    Lifetime=&UniformRangedFloat(Minimum=0.2, Maximum=0.5)
    Mesh Scale=&UniformRangedVector(Minimum=<0.7 0.7 0.7>, Maximum=<1.1 0.7 0.7>)
    Module.Write Color=true
    Module.Write Lifetime=true
    Module.Write Mass=true
    Module.Write Material Random=true
    Module.Write Position=true
    Module.Write Scale=true
    Module.Write SpriteRotation=true
    Module.Write SpriteSize=true
    Position=&SimulationPosition()
    Sprite Rotation=&UniformRangedFloat(Minimum=0, Maximum=0)
    Sprite Size=&Vector2DFromFloat(Value=&UniformRangedFloat(Minimum=1, Maximum=20))
   )
   &AddVelocityInCone(
    Cone Angle=9
    Cone Axis=<0 0 1>
    Cone Axis Coordinate Space=::World
    Module.Use Velocity Falloff On Cone Axis=false
    Velocity Falloff Away From Cone Axis=0
    Velocity Strength=5
   )
  ]
  particleUpdateScript [
   &ParticleState()
   &SolveForcesAndVelocity(Clamp Velocity=false, Limit Acceleration=false)
   &ScaleMeshSize(Scale Factor=&VectorFromCurve(VectorCurve=DataInterface() {propertyValues {
       XCurve {keys [
         <at -0.0011 1.0067 RCIM_Cubic>
         {arriveTangent=-0.7272, interpMode=RCIM_Cubic, leaveTangent=-0.7272, tangentMode=RCTM_User, time=0.9969, value=0.7692}
        ]}
       YCurve {keys [<at -0.0011 1.0067>,<at 0.9989 0.7>]}
       ZCurve {keys [<at -0.0011 1.0067 RCIM_Cubic>,<at 0.9989 0>]}
      }}))
   &Color(
    Color=<rgba 1 1 1 0.15>
    Scale Alpha=&FloatFromCurve(FloatCurve=DataInterface() {propertyValues {Curve {keys [<at -0.0046 0.7 RCIM_Cubic>,<at 1.0011 0.002>]}}})
   )
   &DynamicMaterialParameters(
    Index 0 Param 4=&FloatFromCurve(FloatCurve=DataInterface() {propertyValues {Curve {keys [{arriveTangent=1.9724, time=-0.0046, value=1}, <at 1.0011 0>]}}})
    Module.Param0WriteEnabled=true
    Module.Param1WriteEnabled=true
    Module.Param2WriteEnabled=true
    Module.Param3WriteEnabled=true
   )
  ]
  rendererClasses [&NiagaraMeshRendererProperties]
  renderers [{rendererClass=&NiagaraMeshRendererProperties, rendererIndex=0}]
 }
}
summary {
 systemName=NS_Emote_TossCoin_V3
 emitters [
  {
   emitterName=CameraFacing_Coin_Hologram_FrontSide
   rendererClasses [&NiagaraMeshRendererProperties, &NiagaraLightRendererProperties]
  }
  {emitterName=VelocityFacing_Projection_Ray, rendererClasses [&NiagaraMeshRendererProperties]}
  {emitterName=Flare, rendererClasses [&NiagaraSpriteRendererProperties]}
  {emitterName=Flare001, rendererClasses [&NiagaraSpriteRendererProperties]}
  {emitterName=Flare002, rendererClasses [&NiagaraSpriteRendererProperties]}
  {emitterName=TinySparks, rendererClasses [&NiagaraSpriteRendererProperties]}
  {
   emitterName=CameraFacing_Coin_Hologram_BackSide
   rendererClasses [&NiagaraMeshRendererProperties, &NiagaraLightRendererProperties]
  }
 ]
 userVariables [User.CoinSide {type=NiagaraFloat, defaultValue=0, description=""}]
}
